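# Material function that outputs a procedural mask of a repeating grid of soft-edged, rounded square pixels, filed under Utopia emissive surfaces.
# UVs scaled by Pixel Amount are tiled with Frac, centered, scaled, then shaped by Abs, offset and Saturate.
# Squaring before the distance calculation rounds the corners, and a power driven by Pixel Smooth controls falloff.
# The Pixel Size input is unconnected: the scale reads its preview constant directly.
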MaterialFunction /Hermes/Techart/UtopiaEmissiveSurfaces/MF_SquareGrid1
graph {nodes [
  n1 = Constant() {r=4}
  FunctionInput(Preview=@n1) {bUsePreviewValueAsDefault=true, inputName=Pixel Size, inputType=FunctionInput_Scalar}
  n2 = Multiply(
   A=FunctionInput(Preview=TextureCoordinate()) {bUsePreviewValueAsDefault=true, inputName=UVs, inputType=FunctionInput_Vector2}
   B=FunctionInput(Preview=Constant() {r=10}) {bUsePreviewValueAsDefault=true, inputName=Pixel Amount, inputType=FunctionInput_Scalar}
  )
  n3 = Multiply(A=Add(A=Frac(None=@n2)) {constB=-0.5}, B=@n1) {constB=4}
  n4 = Saturate(None=Add(A=Abs(None=@n3)) {constB=-1})
  n5 = OneMinus(None=Distance(A=Multiply(A=@n4, B=@n4), B=Constant()))
  n6 = Add(A=FunctionInput(Preview=Constant() {r=2}) {bUsePreviewValueAsDefault=true, inputName=Pixel Smooth, inputType=FunctionInput_Scalar})
  FunctionOutput(None=Saturate(None=Power(Base=@n5, Exp=@n6)))
 ]}
