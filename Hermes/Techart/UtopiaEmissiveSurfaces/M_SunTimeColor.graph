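# M_SunTimeColor is a material function that outputs a time-of-day color varying with surface orientation to the sun, apparently for Lumen-related emissive surfaces.
# The sunlight vector from FortniteMaterialParameters is dotted with the world vertex normal and remapped to blend direct and indirect color parameters for morning, day, evening and night.
# Color_Time_of_Day combines the four blends, and the scalar Lumen parameter scales the result.
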MaterialFunction /Hermes/Techart/UtopiaEmissiveSurfaces/M_SunTimeColor
graph {nodes [
  n1 = CollectionParameter() {
   collection=&FortniteMaterialParameters
   parameterName=SunlightDirectionalVector
  }
  n2 = VertexNormalWS()
  n3 = Divide(A=Add(A=DotProduct(A=@n1, B=@n2)))
  n4 = LinearInterpolate(
   A=VectorParameter() {parameterName=LumenColor_Morning, defaultValue=<rgba 0.0918 0.3698 0.3165 1>}.RGB
   B=VectorParameter() {parameterName=LumenColor_Morning_indrect, defaultValue=<rgba 0.3766 0.6997 0.63 1>}.RGB
   Alpha=@n3
  )
  n5 = LinearInterpolate(
   A=VectorParameter() {parameterName=LumenColor_Day, defaultValue=<rgba 1.2 1.0411 0.7312 1>}.RGB
   B=VectorParameter() {parameterName=LumenColor_Day_Indirect, defaultValue=<rgba 0.154 0.4295 0.4635 1>}.RGB
   Alpha=@n3
  )
  n6 = LinearInterpolate(
   A=VectorParameter() {parameterName=LumenColor_Evening, defaultValue=<rgba 0.0857 1 0.8883 1>}.RGB
   B=VectorParameter() {parameterName=LumenColor_Evening_Indirect, defaultValue=<rgba 0.1955 0.4378 0.6962 1>}.RGB
   Alpha=@n3
  )
  n7 = LinearInterpolate(
   A=VectorParameter() {parameterName=LumenColor_Night, defaultValue=<rgba 0.2203 2 0.8185 1>}.RGB
   B=VectorParameter() {parameterName=LumenColor_Night_Indirect, defaultValue=<rgba 0.0322 0.4115 0.3065 1>}.RGB
   Alpha=@n3
  )
  FunctionOutput(None=Multiply(
    A=MaterialFunctionCall(Color_Morning=@n4, Color_Day=@n5, Color_Evening=@n6, Color_Night=@n7) {materialFunction=&Color_Time_of_Day}.Result
    B=ScalarParameter() {defaultValue=0.5, parameterName=Lumen}
   ))
 ]}
