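# Subsurface material for a glassy hologram box that shows a logo or image from a 3x3 atlas.
# Custom primitive data picks the image and side axis, masked by face normals and projected on local-space UVs with view-dependent parallax.
# Pixel grid, animated strips, borders, fresnel, skylight reflection and interior-cubemap fake depth add detail.
# Emissive compensates for eye adaptation, a Mask Maker switch zeroes it, and output passes through destructible damage.
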Material /Hermes/Techart/UtopiaEmissiveSurfaces/M_HoloBox
graph {
 properties {
  bAlwaysEvaluateWorldPositionOffset=true
  bAutomaticallySetUsageInEditor=false
  bUseMaterialAttributes=true
  shadingModel=MSM_Subsurface
  translucencyLightingMode=TLM_Surface
  usedShadingModels=Subsurface
 }
 nodes [
  Add() {constB=-0.05}
  Constant()
  n1 = NamedRerouteUsage()
  Multiply(A=@n1, B=@n1)
  n2 = TextureCoordinate()
  n3 = Divide(A=Time()) {constB=5}
  n4 = Add(A=ComponentMask(None=@n2) {g=true}, B=@n3)
  n5 = TextureSample(UVs=AppendVector(A=ComponentMask(None=@n2) {r=true}, B=Multiply(A=@n4) {constB=0.5})) {samplerType=SAMPLERTYPE_LinearColor, texture=/Game/Characters/Player/Male/Medium/Bodies/M_MED_BiruFang/Textures/T_BiruFang_Noise}
  n6 = OneMinus(None=@n5.G)
  n7 = Sine(None=@n3)
  n8 = LinearInterpolate(A=@n5.G, B=@n6, Alpha=Add(A=Multiply(A=@n7) {constB=0.5}) {constB=0.5})
  n9 = Multiply(A=Multiply(A=@n6, B=Multiply(A=@n5.G, B=@n8))) {constB=4}
  n10 = Saturate(None=LinearInterpolate(Alpha=Saturate(None=@n9)) {constA=0.1, constB=0.9})
  Multiply(A=LinearInterpolate(
    A=VectorParameter() {group=Color, parameterName=Glass Color MASK MAKER 2, defaultValue=<rgba 0.1875 0.0014 0.014 1>}.RGB
    B=VectorParameter() {group=Color, parameterName=Glass Color MASK MAKER1, defaultValue=<rgba 0.1562 0.002 0.0679 1>}.RGB
    Alpha=@n10
   )) {constB=3}
  n11 = ReflectionVectorWS()
  n12 = ScalarParameter() {defaultValue=0.1067, parameterName=Cubemap Rough}
  NamedRerouteDeclaration(None=Multiply(A=NamedRerouteUsage(), B=SkyLightEnvMapSample(Direction=@n11, Roughness=@n12))) {name=Color Cubemap, nodeColor=<rgba 0.0547 0 1 1>}
  n13 = Constant3Vector() {constant=<rgba 0 1 0.2149 1>}
  n14 = Constant3Vector() {constant=<rgba 0.0254 0.2257 1 1>}
  NamedRerouteDeclaration(None=LinearInterpolate(A=LinearInterpolate(A=@n13, B=@n14, Alpha=@n10), B=VectorParameter() {parameterName=Glass Color, defaultValue=<rgba 0 0.9047 0.6639 1>}.RGB) {constAlpha=0.2}) {name=Base color, nodeColor=<rgba 0 0.1797 1 1>}
  n15 = ScalarParameter() {defaultValue=122.0243, group=UV, parameterName=UV Pos X}
  n16 = LocalPosition()
  n17 = CameraVectorWS()
  n18 = Divide(A=Multiply(A=@n17, B=ScalarParameter() {defaultValue=-20, parameterName=Depth Main Shape}), B=DotProduct(A=@n17, B=VertexNormalWS()))
  n19 = MaterialFunctionCall(Float3=Add(A=@n16.XYZ, B=Transform(None=@n18) {transformSourceType=TRANSFORMSOURCE_World, transformType=TRANSFORM_Local})) {materialFunction=&BreakOutFloat3Components}
  n20 = Multiply(A=ScalarParameter() {defaultValue=1, group=UV, parameterName=UV Size X}, B=@n19.R)
  n21 = Multiply(A=Multiply(A=@n19.B, B=Constant() {r=-1}) {constB=0}, B=ScalarParameter() {defaultValue=1, group=UV, parameterName=UV Size Y})
  n22 = ScalarParameter() {defaultValue=119.6498, group=UV, parameterName=UV Pos Y}
  n23 = Divide(A=MaterialFunctionCall(X=Add(A=@n15, B=@n20), Y=@n19.G, Z=Add(A=@n21, B=@n22) {constB=50}) {materialFunction=&MakeFloat3}.Result) {constB=260}
  n24 = ComponentMask(None=@n23) {b=true, r=true}
  NamedRerouteDeclaration(None=@n24) {name=Deep Layer, nodeColor=<rgba 1 0 0.3281 1>}
  n25 = TextureObject() {texture=/Hermes/Techart/UtopiaEmissiveSurfaces/Acad_Stand_Holo/T_HoloAtlasEx}
  n26 = ScalarParameter() {bUseCustomPrimitiveData=true, parameterName=Image, primitiveDataIndex=1}
  n27 = Multiply(
   A=MaterialFunctionCall(
    "Animation  Phase (0-1)"=@n26
    Number of Rows=Constant() {r=3}
    Number of Columns=Constant() {r=3}
    Texture=@n25
    UVs=NamedRerouteUsage()
    MipBias/Level=Constant() {r=3}
   ) {materialFunction=&FlipBook}.Result
   B=NamedRerouteUsage()
  )
  n28 = NamedRerouteUsage()
  NamedRerouteDeclaration(None=Multiply(A=Multiply(A=@n27, B=@n28), B=NamedRerouteUsage())) {name=Shape2, nodeColor=<rgba 0.4062 1 0 1>}
  n29 = VertexInterpolator(VS=PreSkinnedNormal())
  n30 = ComponentMask(None=@n29.PS) {g=true}
  n31 = Abs(None=ComponentMask(None=@n29.PS) {b=true})
  n32 = ComponentMask(None=@n29.PS) {r=true}
  n33 = LinearInterpolate(
   A=Subtract(A=Abs(None=@n30), B=@n31)
   B=Subtract(A=Abs(None=@n32), B=@n31)
   Alpha=ScalarParameter() {bUseCustomPrimitiveData=true, parameterName="Side image (0=Y/ 1=X)"}
  )
  NamedRerouteDeclaration(None=Saturate(None=Power(Base=@n33) {constExponent=4})) {name=Side Mask, nodeColor=<rgba 1 0 0.4453 1>}
  n34 = NamedRerouteUsage()
  n35 = Abs(None=Multiply(A=Add(A=@n34) {constB=-0.5}) {constB=0.5})
  n36 = ScalarParameter() {defaultValue=-0.23, parameterName=Image Border}
  n37 = Distance(A=Saturate(None=Add(A=@n35, B=@n36) {constB=-0.23}), B=Constant())
  NamedRerouteDeclaration(None=OneMinus(None=Sign(None=@n37))) {name=UV Mask, nodeColor=<rgba 0 0.7656 1 1>}
  MaterialExpressionNamedRerouteUsage_30 = NamedRerouteUsage()
  n38 = Convert(None=@MaterialExpressionNamedRerouteUsage_30) {
   nodeName=Break Float2
   convertInputs [{
     type=Vector2
     expressionInput {
      expression=&self:MaterialExpressionNamedRerouteUsage_30
      inputName=None
      mask=0
      maskA=0
      maskB=0
      maskG=0
      maskR=0
      outputIndex=0
     }
    }]
   convertOutputs [{type=Scalar}, {type=Scalar}]
  }
  n39 = Multiply(A=AppendVector(A=Multiply(A=@n38) {constB=2}, B=@n38), B=ScalarParameter() {defaultValue=100, parameterName=Pixel Amount})
  n40 = Multiply(A=Add(A=Frac(None=@n39)) {constB=-0.5}, B=ScalarParameter() {defaultValue=4, parameterName=Pixel Separation}) {constB=4}
  n41 = Saturate(None=Add(A=Abs(None=@n40)) {constB=-1})
  n42 = Multiply(A=@n41, B=@n41)
  n43 = Constant()
  n44 = Power(Base=OneMinus(None=Distance(A=@n42, B=@n43)), Exp=ScalarParameter() {defaultValue=1.5811, parameterName=Pixel Size})
  NamedRerouteDeclaration(None=LinearInterpolate(Alpha=Saturate(None=@n44)) {constA=0.8}) {name=Pixel Mask, nodeColor=<rgba 0.4297 0 1 1>}
  NamedRerouteDeclaration(None=@n23) {name=UVs, nodeColor=<rgba 0 1 0.6484 1>}
  n45 = Add(
   A=Multiply(A=Time(), B=ScalarParameter() {defaultValue=0.1, parameterName=Sine Speed})
   B=Divide(A=WorldPosition().Z, B=ScalarParameter() {defaultValue=10, parameterName=Sine Size}) {constB=12}
  )
  n46 = Add(A=Multiply(A=Sine(None=@n45)) {constB=0.5}) {constB=0.5}
  n47 = Divide(A=Time()) {constB=10}
  n48 = Add(A=Multiply(A=Sine(None=@n47)) {constB=0.5}) {constB=0.5}
  NamedRerouteDeclaration(None=Multiply(A=@n46, B=@n48)) {name=Strips, nodeColor=<rgba 0 0.4141 1 1>}
  n49 = ScalarParameter() {defaultValue=1, parameterName=Opacity DIstant, sliderMax=1}
  n50 = ScalarParameter() {defaultValue=0.8, group=Brand, parameterName=Logo Opacity, sliderMax=1}
  n51 = Saturate(None=@n46)
  NamedRerouteDeclaration(None=LinearInterpolate(B=LinearInterpolate(A=@n49, B=@n50, Alpha=@n51) {constA=0.8}, Alpha=NamedRerouteUsage()) {constA=1}) {name=Strips Color, nodeColor=<rgba 0.0547 1 0 1>}
  NamedRerouteDeclaration(None=@n48) {name=Fade, nodeColor=<rgba 0.3828 0 1 1>}
  n52 = Normalize(VectorInput=Add(A=SkyAtmosphereLightDirection(), B=CameraVectorWS()))
  n53 = VertexNormalWS()
  NamedRerouteDeclaration(None=Power(Base=DotProduct(A=@n52, B=@n53), Exp=Constant() {r=64})) {name=Spec, nodeColor=<rgba 0 1 0.0391 1>}
  n54 = NamedRerouteDeclaration(None=MaterialFunctionCall() {materialFunction=&/M_SunTimeColor}.Result) {name=Sun Color, nodeColor=<rgba 1 0.4637 0 1>}
  n55 = NamedRerouteUsage()
  n56 = Multiply(A=@n54, B=@n55)
  n57 = ScalarParameter() {defaultValue=2.3714, parameterName=Param_6}
  n58 = LinearInterpolate(A=NamedRerouteUsage(), B=@n55)
  n59 = MaterialFunctionCall(Float2=TextureCoordinate()) {materialFunction=&BreakOutFloat2Components}
  n60 = Add(
   A=ScalarParameter() {parameterName=Offset Interior X}
   B=Multiply(A=ScalarParameter() {defaultValue=1, parameterName=Size Interior X}, B=@n59.R)
  )
  n61 = Add(
   A=Multiply(A=@n59.G, B=ScalarParameter() {defaultValue=1, parameterName=Size Interior Y})
   B=ScalarParameter() {parameterName=Offset Interior Y}
  )
  n62 = ScalarParameter() {defaultValue=1, group=B_SSS_EdgeDepth, parameterName=SSSDepth X, sortPriority=5}
  n63 = ScalarParameter() {defaultValue=1, group=B_SSS_EdgeDepth, parameterName=SSSDepth Y, sortPriority=5}
  n64 = ComponentMask(None=MaterialFunctionCall!(UVs=AppendVector(A=@n60, B=@n61), Tiling=AppendVector(A=@n62, B=@n63), Randomize Rotation) {materialFunction=/Engine/Functions/Engine_MaterialFunctions01/Cubemaps/InteriorCubemap}.UVW) {r=true}
  n65 = Constant() {r=0.98}
  n66 = NamedRerouteUsage()
  n67 = LinearInterpolate(
   B=Divide(A=Add(A=@n64, B=@n65), B=ScalarParameter() {defaultValue=0.0803, group=B_SSS_EdgeDepth, parameterName=SSS_Depth Edge Blur, sortPriority=3})
   Alpha=@n66
  )
  n68 = NamedRerouteDeclaration(None=LinearInterpolate(A=ScalarParameter() {group=B_SSS_EdgeDepth, parameterName=SSS Depth Min, sortPriority=4}, Alpha=Saturate(None=@n67))) {name=Depth Mask, nodeColor=<rgba 1 0 0.4453 1>}
  NamedRerouteDeclaration(None=LinearInterpolate(A=Multiply(A=@n56, B=@n57) {constB=0.1}, B=Multiply(A=@n54, B=@n58), Alpha=Saturate(None=@n68))) {name=Fake Depth, nodeColor=<rgba 1 0 0.9375 1>}
  n69 = MaterialFunctionCall(Float3=@n16.XYZ) {materialFunction=&BreakOutFloat3Components}
  n70 = Add(
   A=ScalarParameter() {defaultValue=122.0243, group=UV, parameterName=UV Pos X}
   B=Multiply(A=ScalarParameter() {defaultValue=1, group=UV, parameterName=UV Size X}, B=@n69.R)
  )
  n71 = Constant() {r=-1}
  n72 = Add(
   A=Multiply(A=Multiply(A=@n69.B, B=@n71) {constB=0}, B=ScalarParameter() {defaultValue=1, group=UV, parameterName=UV Size Y})
   B=ScalarParameter() {defaultValue=119.6498, group=UV, parameterName=UV Pos Y}
  ) {constB=50}
  n73 = ComponentMask(None=Divide(A=MaterialFunctionCall(X=@n70, Y=@n69.G, Z=@n72) {materialFunction=&MakeFloat3}.Result) {constB=260}) {b=true, r=true}
  n74 = Multiply(A=Power(Base=NamedRerouteUsage()) {constExponent=8}) {constB=0.005}
  n75 = Divide(A=PixelDepth().R, B=ScalarParameter() {defaultValue=500, group=Dots, parameterName=Dot Texture Distance}) {constB=350}
  n76 = NamedRerouteDeclaration(None=Saturate(None=OneMinus(None=@n75))) {name=Distance Lerp, nodeColor=<rgba 1 0.008 0.0622 1>}
  NamedRerouteDeclaration(None=Add(A=LinearInterpolate(A=@n73, B=@n24) {constAlpha=0.8}, B=LinearInterpolate(B=@n74, Alpha=@n76)) {constB=0}) {name=Mid Layer, nodeColor=<rgba 1 0 0.3281 1>}
  n77 = TextureCoordinate()
  n78 = Multiply(A=@n77, B=OneMinus(None=@n77))
  n79 = Multiply(
   A=Multiply(A=ComponentMask(None=@n78) {r=true}, B=ComponentMask(None=@n78) {g=true})
   B=ScalarParameter() {defaultValue=133, parameterName=Border Size}
  )
  n80 = ScalarParameter() {defaultValue=0.1648, parameterName=Border Gradient}
  n81 = Saturate(None=MaterialFunctionCall(In=Power(Base=@n79, Exp=@n80), Contrast=ScalarParameter() {defaultValue=2, parameterName=Border Contrast}) {materialFunction=&CheapContrast}.Result)
  NamedRerouteDeclaration(None=Multiply(A=OneMinus(None=@n81)) {constB=0.5}) {name=Borde, nodeColor=<rgba 0 0.1328 1 1>}
  NamedRerouteUsage()
  NamedRerouteUsage()
  NamedRerouteUsage()
  NamedRerouteUsage()
  PixelNormalWS()
  n82 = NamedRerouteUsage()
  Saturate(None=@n82)
  n83 = NamedRerouteUsage()
  n84 = OneMinus(None=Saturate(None=NamedRerouteUsage()))
  n85 = Multiply(
   A=Multiply(A=Multiply(A=@n83) {constB=0.2}, B=Add(A=@n84) {constB=0.4})
   B=StaticSwitchParameter(True=Constant() {r=0.3}, False=Constant() {r=0.8}) {group=Projector, parameterName=Mask Maker Take Over, sortPriority=1}
  ) {constB=0.8}
  n86 = Constant()
  n87 = TextureCoordinate()
  n88 = Multiply(
   A=NamedRerouteUsage()
   B=Power(Base=ComponentMask(None=@n87) {g=true}, Exp=ScalarParameter() {defaultValue=4.4028, parameterName=Bottom Gradient Size})
  )
  n89 = ScalarParameter() {defaultValue=1.3383, parameterName=gradient Int}
  n90 = LinearInterpolate(Alpha=@n82) {constA=1, constB=0.8}
  n91 = VectorParameter() {group=Brand, parameterName=Logo Color MASK MAKER, defaultValue=<rgba 0.3594 0.0072 0.0072 1>}
  n92 = VectorParameter() {parameterName=Glass Color2, defaultValue=<rgba 0.0054 0.1619 0.1823 1>}
  n93 = NamedRerouteUsage()
  n94 = Multiply(
   A=StaticSwitchParameter(True=LinearInterpolate(A=@n91.RGB, B=@n92.RGB, Alpha=@n93), False=@n92.RGB) {group=Projector, parameterName=Mask Maker Take Over, sortPriority=1}
   B=ScalarParameter() {defaultValue=7, parameterName=Logo Intensity}
  )
  n95 = MaterialFunctionCall("Animation  Phase (0-1)"=@n26, Number of Rows=Constant() {r=3}, Number of Columns=Constant() {r=3}, Texture=@n25, UVs=NamedRerouteUsage()) {materialFunction=&FlipBook}
  n96 = Multiply(
   A=MaterialFunctionCall(In=@n95.Result, Contrast=Constant() {r=1}) {materialFunction=&CheapContrast}.Result
   B=NamedRerouteUsage()
  )
  n97 = LinearInterpolate(A=Multiply(A=@n95.Result, B=ScalarParameter() {defaultValue=1.2, parameterName=Glow}), B=@n96, Alpha=@n96)
  n98 = NamedRerouteUsage()
  n99 = NamedRerouteDeclaration(None=Multiply(A=Multiply(A=@n97, B=@n98), B=NamedRerouteUsage())) {name=Logo, nodeColor=<rgba 1 0.5156 0 1>}
  n100 = NamedRerouteUsage()
  n101 = NamedRerouteDeclaration(None=Saturate(None=Multiply(A=@n99, B=@n100))) {name=Alpha Mask, nodeColor=<rgba 1 0.0234 0 1>}
  n102 = Saturate(None=NamedRerouteUsage())
  n103 = LinearInterpolate(
   A=Multiply(A=NamedRerouteUsage()) {constB=0.6}
   B=Multiply(A=Multiply(A=@n90, B=@n94), B=LinearInterpolate(A=@n101, Alpha=@n102) {constB=0})
   Alpha=NamedRerouteUsage()
  )
  n104 = NamedRerouteUsage()
  n105 = LinearInterpolate(A=Add(A=Multiply(A=@n88, B=@n89), B=@n103), B=@n103, Alpha=@n104)
  n106 = LinearInterpolate(Alpha=NamedRerouteUsage()) {constB=0.1}
  n107 = Saturate(None=@n1)
  n108 = Add(A=@n105, B=LinearInterpolate(A=LinearInterpolate(A=@n106, Alpha=@n107) {constB=0}, B=@n105, Alpha=@n104) {constB=0})
  n109 = ScalarParameter() {defaultValue=1, parameterName=Fresnel}
  n110 = LinearInterpolate(
   B=Add(A=NamedRerouteUsage(), B=NamedRerouteUsage())
   Alpha=NamedRerouteDeclaration(None=Fresnel(ExponentIn=@n109) {baseReflectFraction=0}) {name=Fresnel, nodeColor=<rgba 0 1 0.5547 1>}
  )
  n111 = EyeAdaptationInverse(LightValueInput=Add(A=Add(A=@n108, B=@n110) {desc=Frenel}, B=NamedRerouteUsage()), AlphaInput=Constant() {r=0.4})
  n112 = Saturate(None=NamedRerouteUsage())
  n113 = MaterialFunctionCall(Material_In=MakeMaterialAttributes!(
    BaseColor=@n85
    Metallic=ScalarParameter() {parameterName=Metallic}
    Specular=Constant() {r=0.5}
    Roughness=ScalarParameter() {defaultValue=0.0533, parameterName=Rough}
    Anisotropy
    EmissiveColor=StaticSwitchParameter(True=@n86, False=@n111.EyeAdaptationInverse) {group=Projector, parameterName=Mask Maker Take Over, sortPriority=1}
    Opacity=OneMinus(None=@n112)
    OpacityMask
    Normal
    Tangent
    WorldPositionOffset
    SubsurfaceColor=@n85
    ClearCoat
    ClearCoatRoughness
    AmbientOcclusion
    Refraction
    CustomizedUVs_0
    CustomizedUVs_1
    CustomizedUVs_2
    CustomizedUVs_3
    CustomizedUVs_4
    CustomizedUVs_5
    CustomizedUVs_6
    CustomizedUVs_7
    PixelDepthOffset
    ShadingModel
    Displacement
   )) {materialFunction=&MF_FN_DestructibleDamage}
  SetMaterialAttributes!(MaterialAttributes=@n113.Material_Out, World Position Offset=@n113.Material_Out) {attributeSetTypes [#1]}
  Sign(None=@n66)
  TextureCoordinate()
  VertexNormalWS()
 ]
}
