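# M_LEDlamp is a Default Lit material for an LED lamp surface, combining lava-lamp bubbles with an LED dot pattern.
# Dithered triplanar UVs drive four counter-scrolling background samples, multiplied into a lava mask that warps tiny bubbles and tints a two-color panel gradient.
# A procedural dot grid uses view-based parallax and is modulated by panning noise.
# Fresnel, roughness and normal maps, and a skylight cubemap reflection supply surface detail.
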Material /Hermes/Techart/UtopiaEmissiveSurfaces/M_LEDlamp
graph {
 properties {bAutomaticallySetUsageInEditor=false, usedShadingModels=Default Lit}
 nodes [
  n1 = VertexNormalWS()
  Abs(None=@n1)
  n2 = VectorParameter() {group=1. Colors, parameterName=Panel Color Down, sortPriority=1, defaultValue=<rgba 0.011 0.0497 0.3763 1>}
  n3 = NamedRerouteDeclaration(None=SkyLightEnvMapSample(Direction=ReflectionVectorWS(), Roughness=ScalarParameter() {defaultValue=0.0493, parameterName=s})) {name=Cubemap, nodeColor=<rgba 0.0547 0 1 1>}
  Add(A=@n2.RGB, B=@n3)
  n4 = ScalarParameter() {defaultValue=1.2271, parameterName=Param_6}
  n5 = TextureCoordinate()
  n6 = Time()
  n7 = TextureSample(UVs=Panner(Coordinate=Multiply(A=@n4, B=@n5), Time=Multiply(A=@n6) {constB=2}) {speedX=0.1, speedY=0.1}) {samplerType=SAMPLERTYPE_LinearColor, texture=/Game/Characters/Player/Male/Medium/Bodies/M_MED_BiruFang/Textures/T_BiruFang_Noise}
  Clamp(None=Multiply(A=@n7.G) {constB=2}) {maxDefault=10, minDefault=1}
  Constant()
  n8 = VectorParameter() {group=1. Colors, parameterName=Buble Color, sortPriority=1, defaultValue=<rgba 0.1051 0.6141 0.7865 1>}
  n9 = Multiply(A=NamedRerouteUsage()) {constB=1.4}
  n10 = Multiply(A=ComponentMask(None=@n9) {r=true}) {constB=-1}
  n11 = Multiply(A=Time() {bOverride_Period=true, period=333.3333}) {constB=0.003}
  n12 = ComponentMask(None=@n9) {g=true}
  n13 = Multiply(A=Time() {bOverride_Period=true, period=100}) {constB=0.01}
  n14 = TextureSample!(
   UVs=AppendVector(A=Add(A=@n10, B=@n11), B=Add(A=@n12, B=@n13))
   Tex
   "DDX(UVs)"=NamedRerouteUsage()
   "DDY(UVs)"=NamedRerouteUsage()
   Apply View MipBias
  ) {
   mipValueMode=TMVM_Derivative
   samplerType=SAMPLERTYPE_LinearGrayscale
   texture=/Game/Environments/Asteria/Sets/Lair/Materials/LavaLamp/T_LavaLamp_Background
  }
  n15 = NamedRerouteUsage()
  n16 = Multiply(A=@n15) {constB=-1.3}
  n17 = ComponentMask(None=@n16) {r=true}
  n18 = Multiply(A=Time() {bOverride_Period=true, period=222.2222}) {constB=-0.0045}
  n19 = Multiply(A=ComponentMask(None=@n16) {g=true}) {constB=-1}
  n20 = Multiply(A=Time() {bOverride_Period=true, period=66.6667}) {constB=-0.015}
  n21 = TextureSample(UVs=AppendVector(A=Add(A=@n17, B=@n18), B=Add(A=@n19, B=@n20)), "DDX(UVs)"=NamedRerouteUsage(), "DDY(UVs)"=NamedRerouteUsage()) {
   mipValueMode=TMVM_Derivative
   samplerType=SAMPLERTYPE_LinearGrayscale
   texture=&T_LavaLamp_Background
  }
  n22 = Multiply(A=@n14.R, B=@n21.R)
  n23 = Multiply(A=NamedRerouteUsage()) {constB=1.2}
  n24 = ComponentMask(None=@n23) {r=true}
  n25 = Multiply(A=Time() {bOverride_Period=true, period=166.6667}) {constB=0.006}
  n26 = ComponentMask(None=@n23) {g=true}
  n27 = Multiply(A=Time() {bOverride_Period=true, period=50}) {constB=0.02}
  n28 = TextureSample(UVs=AppendVector(A=Add(A=@n24, B=@n25), B=Add(A=@n26, B=@n27)), "DDX(UVs)"=NamedRerouteUsage(), "DDY(UVs)"=NamedRerouteUsage()) {
   mipValueMode=TMVM_Derivative
   samplerType=SAMPLERTYPE_LinearGrayscale
   texture=&T_LavaLamp_Background
  }
  n29 = NamedRerouteUsage()
  n30 = Multiply(A=@n29) {constB=1.1}
  n31 = Add(A=ComponentMask(None=@n30) {r=true}, B=Multiply(A=Time() {bOverride_Period=true, period=151.5152}) {constB=-0.0066})
  n32 = Add(A=ComponentMask(None=@n30) {g=true}, B=Multiply(A=Time() {bOverride_Period=true, period=45.4545}) {constB=-0.022})
  n33 = TextureSample(UVs=Multiply(A=AppendVector(A=@n31, B=@n32)) {constB=-1}, "DDX(UVs)"=NamedRerouteUsage(), "DDY(UVs)"=NamedRerouteUsage()) {
   mipValueMode=TMVM_Derivative
   samplerType=SAMPLERTYPE_LinearGrayscale
   texture=&T_LavaLamp_Background
  }
  n34 = Multiply(A=@n28.R, B=@n33.R)
  n35 = Saturate(None=Multiply(A=Multiply(A=@n22, B=@n34)) {constB=2.5})
  n36 = Multiply(A=@n35, B=@n35)
  n37 = Multiply(A=@n36, B=@n36)
  n38 = OneMinus(None=@n37)
  LinearInterpolate(B=@n8.RGB, Alpha=@n38) {constA=1}
  n39 = NamedRerouteUsage()
  n40 = CameraVectorWS()
  n41 = ScalarParameter() {parameterName=Param_2}
  n42 = VertexNormalWS()
  n43 = Add(A=AppendVector(A=Divide(A=@n39), B=Constant()), B=Divide(A=Multiply(A=@n40, B=@n41), B=DotProduct(A=@n40, B=@n42)))
  n44 = Add(A=Frac(None=ComponentMask(None=@n43) {g=true, r=true})) {constB=-0.5}
  n45 = MaterialFunctionCall(Float2=Multiply(A=@n44, B=@n44)) {materialFunction=&BreakOutFloat2Components}
  n46 = Multiply(A=Add(A=@n45.R, B=@n45.R), B=ScalarParameter() {defaultValue=58.8548, parameterName=Shape Mult}) {constB=90}
  n47 = Multiply(A=ScalarParameter() {defaultValue=3, parameterName=Shape Sub}) {constB=-1}
  n48 = OneMinus(None=Saturate(None=Add(A=@n46, B=@n47) {constB=-2}))
  n49 = Add(A=Frac(None=ComponentMask(None=@n43) {b=true, g=true})) {constB=-0.5}
  n50 = MaterialFunctionCall(Float2=Multiply(A=@n49, B=@n49)) {materialFunction=&BreakOutFloat2Components}
  n51 = Multiply(A=Add(A=@n50.R, B=@n50.R), B=ScalarParameter() {defaultValue=58.8548, parameterName=Shape Mult}) {constB=90}
  n52 = Multiply(A=ScalarParameter() {defaultValue=3, parameterName=Shape Sub}) {constB=-1}
  n53 = OneMinus(None=Saturate(None=Add(A=@n51, B=@n52) {constB=-2}))
  n54 = Add(A=Frac(None=ComponentMask(None=@n43) {b=true, r=true})) {constB=-0.5}
  n55 = MaterialFunctionCall(Float2=Multiply(A=@n54, B=@n54)) {materialFunction=&BreakOutFloat2Components}
  n56 = Add(A=@n55.R, B=@n55.R)
  n57 = ScalarParameter() {defaultValue=58.8548, parameterName=Shape Mult}
  n58 = ScalarParameter() {defaultValue=3, parameterName=Shape Sub}
  n59 = Saturate(None=Add(A=Multiply(A=@n56, B=@n57) {constB=90}, B=Multiply(A=@n58) {constB=-1}) {constB=-2})
  n60 = ScalarParameter() {defaultValue=0.1794}
  n61 = ComponentMask(None=@n1) {b=true}
  n62 = LinearInterpolate(A=Add(A=Add(A=@n48, B=@n53), B=OneMinus(None=@n59)), B=@n48, Alpha=Saturate(None=Step(Y=@n60, X=@n61)))
  n63 = Multiply(A=ScalarParameter() {defaultValue=1.2423, parameterName=Param_1}, B=@n62)
  n64 = LinearInterpolate(B=Saturate(None=Step(Y=@n62)), Alpha=@n62) {constA=1}
  LinearInterpolate(
   A=LinearInterpolate(A=@n63, B=LinearInterpolate(Alpha=@n64) {constA=1.2, constB=0.8}, Alpha=Saturate(None=@n62))
   B=@n63
   Alpha=Saturate(None=NamedRerouteUsage())
  ) {constA=1}
  n65 = Multiply(A=OneMinus(None=@n35), B=ScalarParameter() {defaultValue=15, group=2. Background, parameterName=Background Normal Intensity, sortPriority=6}) {constB=20}
  n66 = Multiply(A=Multiply(
    A=@n38
    B=MaterialFunctionCall!(Fresnel=@n37, Height=@n65) {materialFunction=/Game/Environments/Asteria/Sets/Lair/Materials/LavaLamp/MF_LavaLamp_Colors}.Bubble Warp
   )) {constB=0.8}
  LinearInterpolate(A=@n66, B=Multiply(A=Multiply()) {constB=0.8})
  n67 = VectorParameter() {group=1. Colors, parameterName=Panel Color Up, sortPriority=0, defaultValue=<rgba 0.2122 0.7605 0.8879 1>}
  n68 = VectorParameter() {group=1. Colors, parameterName=Panel Color Down, sortPriority=1, defaultValue=<rgba 0.011 0.0497 0.3763 1>}
  n69 = ComponentMask(None=TextureCoordinate()) {g=true}
  n70 = ScalarParameter() {defaultValue=0.576, parameterName=Param_5}
  n71 = Saturate(None=Power(Base=Subtract(A=@n69, B=@n70), Exp=ScalarParameter() {defaultValue=2.0351, parameterName=Param_4}) {constExponent=3})
  n72 = ScalarParameter() {defaultValue=0.7, group=2. Background, parameterName=Background Alpha, sortPriority=7}
  n73 = Saturate(None=Multiply(A=@n38) {constB=7})
  n74 = Add(A=NamedRerouteUsage(), B=@n66)
  n75 = Multiply(
   A=LinearInterpolate(A=LinearInterpolate(A=@n67.RGB, B=@n68.RGB, Alpha=@n71), B=@n8.RGB, Alpha=Multiply(A=@n72, B=@n73) {constB=0.6})
   B=Multiply(A=TextureSample!(UVs=@n74, Tex, Apply View MipBias) {samplerType=SAMPLERTYPE_LinearGrayscale, texture=/Game/Environments/Asteria/Sets/Lair/Materials/LavaLamp/T_LavaLamp_TinyBubbles}.R) {constB=2}
  )
  LinearInterpolate(A=@n75, B=Multiply(A=@n2.RGB, B=@n3), Alpha=NamedRerouteUsage())
  n76 = NamedRerouteUsage()
  n77 = Multiply(A=TextureCoordinate(), B=ScalarParameter() {defaultValue=2.5, parameterName=Param_8})
  n78 = TextureSampleParameter2D(UVs=@n77) {
   group=5. Textures
   parameterName=Roughness
   samplerType=SAMPLERTYPE_LinearColor
   sortPriority=0
   texture=/Game/Environments/Asteria/Sets/Lair/Materials/LavaLamp/T_LavaLamp_R
  }
  Multiply(A=Multiply(A=@n75, B=@n76), B=Clamp(None=OneMinus(None=@n78.G)) {minDefault=0.8})
  n79 = Multiply(A=@n29) {constB=4}
  n80 = Multiply(A=ComponentMask(None=@n79) {g=true}) {constB=-1}
  n81 = Multiply(A=Time()) {constB=0.32}
  NamedRerouteDeclaration(None=AppendVector(A=ComponentMask(None=@n79) {r=true}, B=Add(A=@n80, B=@n81))) {name=TinyBubblesCoord2, nodeColor=<rgba 0 0.7891 1 1>}
  n82 = Multiply(A=@n15, B=ScalarParameter() {defaultValue=4.8268, parameterName=Tiny bubles UV}) {constB=10}
  n83 = Multiply(A=ComponentMask(None=@n82) {g=true}) {constB=-1}
  n84 = Multiply(A=Time()) {constB=0.2}
  NamedRerouteDeclaration(None=AppendVector(A=ComponentMask(None=@n82) {r=true}, B=Add(A=@n83, B=@n84))) {name=TinyBubblesCoord1, nodeColor=<rgba 1 0.4688 0 1>}
  n85 = MaterialFunctionCall!(
   "Normal (WS)"
   Texture Scale=ScalarParameter() {defaultValue=4, parameterName=Triplanar Scale}
   Triplanar Fade=ScalarParameter() {defaultValue=10, parameterName=Triplanar Fade}
   Derivatives Scale
  ) {materialFunction=/Game/Athena/Environments/MaterialFunctions/MF_TriplanarDithered}
  NamedRerouteDeclaration(None=@n85.DDX) {name=DDX, nodeColor=<rgba 1 0 0.2578 1>}
  NamedRerouteDeclaration(None=@n85.DDY) {name=DDY, nodeColor=<rgba 1 0 0.2578 1>}
  n86 = Multiply(A=@n7.G) {constB=2.5}
  NamedRerouteDeclaration(None=LinearInterpolate(B=Clamp(None=@n86) {maxDefault=10, minDefault=1.1}, Alpha=@n62) {constA=1, constB=1.2}) {name=Dotsa, nodeColor=<rgba 0 0.7422 1 1>}
  NamedRerouteDeclaration(None=@n38) {name=Lava, nodeColor=<rgba 0.9922 1 0 1>}
  NamedRerouteDeclaration(None=Fresnel(ExponentIn=ScalarParameter() {defaultValue=1, parameterName=Fresnel}) {baseReflectFraction=0}) {name=Fresnel, nodeColor=<rgba 0 1 0.5547 1>}
  n87 = Divide(A=NamedRerouteDeclaration(None=@n85.UVs) {name=Triplanar UV, nodeColor=<rgba 1 0 0.2578 1>}) {constB=200}
  n88 = AppendVector(
   A=ScalarParameter() {defaultValue=1.2, group=2. Background, parameterName=Background Tiling X, sortPriority=0}
   B=ScalarParameter() {defaultValue=0.6, group=2. Background, parameterName=Background Tiling Y, sortPriority=1}
  )
  NamedRerouteDeclaration(None=Multiply(A=Add(A=@n87) {constB=-0.5}, B=Multiply(A=@n88) {constB=-1})) {name=UVs, nodeColor=<rgba 1.9507 0 2 1>}
  NamedRerouteDeclaration(None=Fresnel(ExponentIn=ScalarParameter() {defaultValue=1, parameterName=Fresnel}) {baseReflectFraction=0}) {name=Fresnel_1, nodeColor=<rgba 0 1 0.5547 1>}
  OneMinus(None=@n7.G)
  Saturate(None=@n76)
  TextureSampleParameter2D(UVs=@n77) {
   group=5. Textures
   parameterName=Normal
   samplerType=SAMPLERTYPE_Normal
   sortPriority=3
   texture=/Game/Environments/Asteria/Sets/Lair/Materials/LavaLamp/T_LavaLamp_N
  }
 ]
}
