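# Material function named as a dot grid that builds a repeating, soft-edged mask from UVs, tiled by Pixel Amount.
# It centers each cell with Frac and squares the offset for a falloff scaled by Pixel Size.
# Pixel Smooth adds a scaled bias before saturate and invert.
# Both summed falloff terms read the R component, so as wired the mask varies only along U, giving stripes rather than round dots.
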MaterialFunction /Hermes/Techart/UtopiaEmissiveSurfaces/MF_DotGrid
graph {nodes [
  n1 = FunctionInput(Preview=TextureCoordinate()) {bUsePreviewValueAsDefault=true, inputName=UVs, inputType=FunctionInput_Vector2}
  n2 = FunctionInput(Preview=Constant() {r=10}) {bUsePreviewValueAsDefault=true, inputName=Pixel Amount, inputType=FunctionInput_Scalar}
  n3 = Add(A=Frac(None=Multiply(A=@n1, B=@n2))) {constB=-0.5}
  n4 = MaterialFunctionCall(Float2=Multiply(A=@n3, B=@n3)) {materialFunction=&BreakOutFloat2Components}
  n5 = Constant() {r=4}
  n6 = Constant() {r=2}
  n7 = Add(
   A=Multiply(A=Add(A=@n4.R, B=@n4.R), B=FunctionInput(Preview=@n5) {bUsePreviewValueAsDefault=true, inputName=Pixel Size, inputType=FunctionInput_Scalar}) {constB=90}
   B=Multiply(A=FunctionInput(Preview=@n6) {bUsePreviewValueAsDefault=true, inputName=Pixel Smooth, inputType=FunctionInput_Scalar}) {constB=0.2}
  ) {constB=-2}
  FunctionOutput(None=OneMinus(None=Saturate(None=@n7)))
 ]}
