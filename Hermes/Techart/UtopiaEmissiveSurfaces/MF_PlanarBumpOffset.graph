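# Material function that offsets UVs with view-dependent parallax, making a texture appear on a flat plane at a set depth from the surface.
# It scales the world camera vector by Depth, divided by its dot product with the vertex normal, to find the ray-plane intersection.
# That offset moves to tangent space and adds to the UVs, outputting XY.
# It uses no height map, and Depth defaults to negative.
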MaterialFunction /Hermes/Techart/UtopiaEmissiveSurfaces/MF_PlanarBumpOffset
graph {nodes [
  n1 = AppendVector(A=FunctionInput(Preview=TextureCoordinate()) {bUsePreviewValueAsDefault=true, inputName=UVs, inputType=FunctionInput_Vector2}, B=Constant())
  n2 = CameraVectorWS()
  n3 = FunctionInput(Preview=Constant() {r=-0.26}) {bUsePreviewValueAsDefault=true, inputName=Depth, inputType=FunctionInput_Scalar}
  n4 = VertexNormalWS()
  n5 = Transform(None=Divide(A=Multiply(A=@n2, B=@n3), B=DotProduct(A=@n2, B=@n4))) {transformSourceType=TRANSFORMSOURCE_World, transformType=TRANSFORM_Tangent}
  FunctionOutput(None=ComponentMask(None=Add(A=@n1, B=@n5)) {g=true, r=true})
 ]}
