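# Unlit emissive material for a screen building surface, likely in Utopia City, showing a masked screen graphic over an animated dot-matrix display.
# The screen mask is hue-shifted and banded, then added to bump-offset dots lit by two panning noise samples.
# Reroutes define sun specular, Fresnel and an InteriorCubemap-derived edge depth mask; the final blend passes through EyeAdaptationInverse.
# A box-projected UV chain from pre-skinned position and normal is left unconnected.
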Material /Hermes/Techart/UtopiaEmissiveSurfaces/M_ScreenBuilding
graph {
 properties {bAutomaticallySetUsageInEditor=false, shadingModel=MSM_Unlit, usedShadingModels=Unlit}
 nodes [
  Add(A=TextureCoordinate()) {constB=-0.5}
  n1 = Divide(A=VertexInterpolator(VS=PreSkinnedPosition()).PS, B=ScalarParameter() {defaultValue=250, parameterName=s}) {constB=200}
  n2 = Abs(None=VertexInterpolator(VS=PreSkinnedNormal()).PS)
  n3 = LinearInterpolate(B=ComponentMask(None=@n1) {b=true, desc=X R, g=true}, Alpha=ComponentMask(None=@n2) {r=true})
  n4 = ComponentMask(None=@n1) {b=true, desc=Y G, r=true}
  n5 = ComponentMask(None=@n2) {g=true}
  Add(A=LinearInterpolate(
    A=LinearInterpolate(A=@n3, B=@n4, Alpha=@n5)
    B=ComponentMask(None=@n1) {desc=Z B, g=true, r=true}
    Alpha=ComponentMask(None=@n2) {b=true, desc=RG}
   )) {constB=0.5}
  Constant() {r=1}
  n6 = TextureSample() {samplerType=SAMPLERTYPE_LinearGrayscale, texture=/Hermes/Sets/UtopiaCity/Textures/T_UtopiaCity_Puppuccino_Screen_M}
  n7 = ScalarParameter() {defaultValue=0.6347, parameterName=Param_3}
  n8 = ScalarParameter() {defaultValue=0.448, parameterName=Param_4}
  n9 = Saturate(None=Divide(A=Subtract(A=@n6.G, B=@n7), B=Subtract(A=@n7, B=@n8)))
  n10 = MaterialFunctionCall(
   Hue Shift Percentage=ScalarParameter() {defaultValue=1, parameterName=Param_6}
   Texture=Multiply(A=@n9, B=Constant3Vector() {constant=<rgba 0.0169 0.6818 1 1>})
  ) {materialFunction=&HueShift}
  n11 = ScalarParameter() {defaultValue=0.9711, parameterName=Param_2}
  n12 = Frac(None=MaterialFunctionCall(In=Power(Base=@n9, Exp=@n11), Contrast=ScalarParameter() {defaultValue=0.408, parameterName=Param_1}) {materialFunction=&CheapContrast}.Result)
  n13 = ScalarParameter() {defaultValue=13.8741, parameterName=Param_5}
  n14 = Saturate(None=Round(None=Multiply(A=@n12, B=@n13)))
  n15 = Multiply(A=NamedRerouteUsage()) {constB=0.6}
  n16 = MaterialFunctionCall!(Depth=ScalarParameter() {defaultValue=-0.2, parameterName="Depth (S)"}, UVs=TextureCoordinate()) {materialFunction=&/MF_PlanarBumpOffset}
  n17 = Multiply(A=@n16.Result, B=ScalarParameter() {defaultValue=0.4759, parameterName=Param_10})
  n18 = Divide(A=Time()) {constB=3}
  n19 = Multiply(
   A=TextureSample(UVs=Panner(Coordinate=@n17, Time=@n18) {speedX=-0.1, speedY=-0.1}) {samplerType=SAMPLERTYPE_LinearColor, texture=/Game/Characters/Player/Male/Medium/Bodies/M_MED_BiruFang/Textures/T_BiruFang_Noise}.G
   B=TextureSample(UVs=Panner(Coordinate=@n17, Time=@n18) {speedX=-0.2, speedY=0.15}) {samplerType=SAMPLERTYPE_LinearColor, texture=&T_BiruFang_Noise}.G
  )
  n20 = LinearInterpolate(Alpha=Saturate(None=@n19)) {constA=3, constB=0.2}
  n21 = NamedRerouteDeclaration(None=VectorParameter() {defaultValue=<rgba 0 0.1647 0.6719 1>}.RGB) {name=Base Color, nodeColor=<rgba 0.0453 0.516 1 1>}
  n22 = Saturate(None=@n19)
  n23 = MaterialFunctionCall!(
   Pixel Amount=ScalarParameter() {defaultValue=15, parameterName="Pixel Amount (S)"}
   Pixel Size=ScalarParameter() {defaultValue=24, parameterName="Pixel Size (S)"}
   Pixel Smooth=ScalarParameter() {defaultValue=1, parameterName="Pixel Smooth (S)"}
   UVs=@n16.Result
  ) {materialFunction=&/MF_DotGrid}
  n24 = LinearInterpolate(B=Multiply(A=LinearInterpolate(Alpha=@n22) {constA=-0.2, constB=4}) {constB=4}, Alpha=@n23.Result)
  n25 = Add(
   A=LinearInterpolate(A=@n9, B=Multiply(A=@n10.Result, B=@n14), Alpha=@n14)
   B=LinearInterpolate(A=Multiply(A=@n15, B=@n20), B=Multiply(A=@n21, B=@n24), Alpha=@n23.Result)
  )
  n26 = Multiply(A=NamedRerouteUsage(), B=NamedRerouteUsage())
  n27 = NamedRerouteUsage()
  n28 = Multiply(A=Constant3Vector() {constant=<rgba 0.2552 0.5587 1 1>}, B=ScalarParameter() {defaultValue=2, parameterName=2223})
  n29 = Multiply(A=@n27, B=OneMinus(None=NamedRerouteUsage()))
  EyeAdaptationInverse(
   LightValueInput=LinearInterpolate(A=Add(A=@n25, B=@n26), B=Multiply(A=@n27, B=@n28), Alpha=Saturate(None=@n29)) {constA=0.1}
   AlphaInput=Constant() {r=0.3}
  )
  NamedRerouteDeclaration(None=@n23.Result) {name=Dots, nodeColor=<rgba 0 1 0.2266 1>}
  n30 = Normalize(VectorInput=Add(A=SkyAtmosphereLightDirection(), B=CameraVectorWS()))
  n31 = VertexNormalWS()
  NamedRerouteDeclaration(None=Power(Base=DotProduct(A=@n30, B=@n31), Exp=Constant() {r=64})) {name=Spec, nodeColor=<rgba 0 1 0.0391 1>}
  n32 = ScalarParameter() {defaultValue=2, parameterName=Fresnel}
  NamedRerouteDeclaration(None=Add(A=NamedRerouteUsage(), B=Fresnel(ExponentIn=@n32) {baseReflectFraction=0})) {name=Fresnel, nodeColor=<rgba 0 1 0.5547 1>}
  n33 = ScalarParameter() {group=B_SSS_EdgeDepth, parameterName=SSS Depth Min, sortPriority=4}
  n34 = TextureCoordinate()
  n35 = ComponentMask(None=@n34) {r=true}
  n36 = Add(
   A=Multiply(A=ComponentMask(None=@n34) {g=true}, B=ScalarParameter() {defaultValue=0.9318, parameterName=Param_8}) {constB=2}
   B=ScalarParameter() {defaultValue=0.064, parameterName=Param_9}
  )
  n37 = ScalarParameter() {defaultValue=1, group=B_SSS_EdgeDepth, parameterName=SSSDepth X, sortPriority=5}
  n38 = ScalarParameter() {defaultValue=1, group=B_SSS_EdgeDepth, parameterName=SSSDepth Y, sortPriority=5}
  n39 = ComponentMask(None=MaterialFunctionCall!(UVs=AppendVector(A=@n35, B=@n36), Tiling=AppendVector(A=@n37, B=@n38), Randomize Rotation) {materialFunction=/Engine/Functions/Engine_MaterialFunctions01/Cubemaps/InteriorCubemap}.UVW) {r=true}
  n40 = Constant() {r=0.98}
  n41 = Saturate(None=Divide(A=Add(A=@n39, B=@n40), B=ScalarParameter() {defaultValue=0.0803, group=B_SSS_EdgeDepth, parameterName=SSS_Depth Edge Blur, sortPriority=3}))
  NamedRerouteDeclaration(None=OneMinus(None=LinearInterpolate(A=@n33, Alpha=@n41))) {name=Depth Mask, nodeColor=<rgba 1 0 0.4453 1>}
  NamedRerouteUsage()
  Power(Base=@n2, Exp=ScalarParameter() {defaultValue=1})
 ]
}
