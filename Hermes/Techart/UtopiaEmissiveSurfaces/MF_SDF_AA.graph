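# Material function that converts a signed distance field input into an antialiased 0-1 mask.
# It divides the SDF by the sum of its absolute screen-space X and Y derivatives, then saturates, giving roughly pixel-wide edges with positive values as the filled side.
# The input's preview is a tiled circle SDF built from scaled, fractional texture coordinates.
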MaterialFunction /Hermes/Techart/UtopiaEmissiveSurfaces/MF_SDF_AA
graph {nodes [
  n1 = TextureCoordinate()
  n2 = Distance(A=Frac(None=Multiply(A=@n1) {constB=2}), B=Constant() {r=0.5})
  n3 = FunctionInput(Preview=Add(A=OneMinus(None=@n2)) {constB=-0.7}) {inputName=SDF}
  n4 = Add(A=Abs(None=DDX(Value=@n3)), B=Abs(None=DDY(Value=@n3)))
  FunctionOutput(None=Saturate(None=Divide(A=@n3, B=@n4)))
 ]}
